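; Skyfire trigger actor that warns players approaching an energy grid and plays the elimination impact when they hit it.
; Overlap volumes toggle a looping rez-out cue, start local sound and bind movement updates that map pawn height to an audio-shape parameter.
; Sound-box impact plays penalty audio and voice, then restarts the audio shape.
; Grid impact runs on server and predicting client, debounced locally, adding ripple, material flash and camera effect, then hiding and stopping the pawn.
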(event EventBeginPlay
  (bind _fortaudioshapebox (Variables|Default|GetFortAudioShapeBox))
  ;[0 0.7007 1] INitalize the Audio Shape
  (|CallParentFunction)
  (Control|SetBoxTransform (Variables|Default|GetFortAudioShapeBox) (Math|Transform|MakeTransform (.location (Transformation|GetWorldTransform (Variables|Default|GetEliminateSoundBoxComponent))) (.rotation (Transformation|GetWorldTransform (Variables|Default|GetEliminateSoundBoxComponent))) (Math|Vector|MakeVector (* (.x (.scale (Transformation|GetWorldTransform (Variables|Default|GetEliminateSoundBoxComponent)))) 2.5e+02) (* (.y (.scale (Transformation|GetWorldTransform (Variables|Default|GetEliminateSoundBoxComponent)))) 2.5e+02) (.z (.scale (Transformation|GetWorldTransform (Variables|Default|GetEliminateSoundBoxComponent)))))))
  (Shape|Enable _fortaudioshapebox)
  (Audio|Parameter|SetFloatParameter (Control|GetEdgeAudioComponent _fortaudioshapebox) (Variables|Sound|GetSFX_Parameter_Name) (Variables|Sound|GetDefault_Parameter_Value) :!class /Script/AudioExtensions.AudioParameterControllerInterface))

(event OnComponentBeginOverlap(Box)
  (params (OverlappedComponent :type Primitive%20Component%20Object%20Reference) (OtherActor :type Actor%20Object%20Reference) (OtherComp :type Primitive%20Component%20Object%20Reference) (OtherBodyIndex :type Integer) (bFromSweep :type Boolean) (SweepResult :type Hit%20Result%20Structure%20%28by%20ref%29))
  ;[0 0.1792 1] Add/remove the rez out gameplay cue when player enters/exits this trigger.
  (bind _asfort_player_pawn (Utilities|Casting|CastToFortPlayerPawn OtherActor))
  (Projectile|AddGameplayCue%28Looping%29Local _asfort_player_pawn (Variables|Default|GetGC_RezOut) (Ability|GameplayCue|MakeGameplayCueParameters 0.0))
  ;[1 0.04536 0.7529] Play Rez out SFX when player hits this trigger
  (Networking|CanExecuteCosmeticEvents
    (:True
      (Utilities|FlowControl|DoOnce false
        (:Completed
          (bind _player_pawn (Variables|Default|GetPlayer%20Pawn)
            (:then
              (if (HelperFunctions|IsLocalPawn _player_pawn)
                (bind _returnvalue (Audio|SpawnSoundAttached (Variables|Sound|GetSFX_RezOut) (Variables|/Script/Engine.Character|GetMesh _player_pawn)))
                (Variables|Sound|SetSFX_RezOut_AC _returnvalue)))
            (:else)))))))

(event Custom|Reset_RezSpawnSound
  ;[1 0.04536 0.7529] Play Rez out SFX when player hits this trigger
  (Utilities|FlowControl|DoOnce false
    (:Completed
      (bind _player_pawn (Variables|Default|GetPlayer%20Pawn)
        (:then
          (if (HelperFunctions|IsLocalPawn _player_pawn)
            (bind _returnvalue (Audio|SpawnSoundAttached (Variables|Sound|GetSFX_RezOut) (Variables|/Script/Engine.Character|GetMesh _player_pawn)))
            (Variables|Sound|SetSFX_RezOut_AC _returnvalue)))
        (:else)))))

(event OnComponentEndOverlap(Box)
  (params (OverlappedComponent :type Primitive%20Component%20Object%20Reference) (OtherActor :type Actor%20Object%20Reference) (OtherComp :type Primitive%20Component%20Object%20Reference) (OtherBodyIndex :type Integer))
  ;[0 0.1792 1] Add/remove the rez out gameplay cue when player enters/exits this trigger.
  (bind _asfort_player_pawn (Utilities|Casting|CastToFortPlayerPawn OtherActor))
  (Projectile|RemoveGameplayCue%28Looping%29Local _asfort_player_pawn (Variables|Default|GetGC_RezOut) (Ability|GameplayCue|MakeGameplayCueParameters 0.0)))

(event Custom|Event%20On%20Character%20Movement%20Update
  (params (DeltaSeconds :type Float%20%28single-precision%29) (OldLocation :type Vector) (OldVelocity :type Vector))
  ;[1 0 0.8302] CHange the sound based on how the character is moving. Using Charcter's current height location on the map to determind min/max
  (bind _player_pawn (Variables|Default|GetPlayer%20Pawn)
    (:then
      (Audio|Parameter|SetFloatParameter (Control|GetEdgeAudioComponent (Variables|Default|GetFortAudioShapeBox)) (Variables|Sound|GetSFX_Parameter_Name) (Math|Float|Clamp(Float) (/ (- (.z (Transformation|GetWorldLocation (Variables|/Script/Engine.Character|GetMesh _player_pawn))) (Variables|Sound|GetSFX_Min_Height)) (- (Variables|Sound|GetSFX_Max_Height) (Variables|Sound|GetSFX_Min_Height)))) :!class /Script/AudioExtensions.AudioParameterControllerInterface))
    (:else)))

(event OnComponentBeginOverlap(EliminateSoundBoxComponent)
  (params (OverlappedComponent :type Primitive%20Component%20Object%20Reference) (OtherActor :type Actor%20Object%20Reference) (OtherComp :type Primitive%20Component%20Object%20Reference) (OtherBodyIndex :type Integer) (bFromSweep :type Boolean) (SweepResult :type Hit%20Result%20Structure%20%28by%20ref%29))
  ;[1 0.02585 0.7297] Play Impact VFX
  (bind _asfort_player_pawn (Utilities|Casting|CastToFortPlayerPawn OtherActor))
  (CallFunction|PlayImpactSounds _asfort_player_pawn))

(event Custom|Play%20Impact%20Sounds
  (params (Player :type Fort%20Player%20Pawn%20Object%20Reference%20%28by%20ref%29))
  (bind _fortaudioshapebox (Variables|Default|GetFortAudioShapeBox))
  ;[1 0 0.8076] Play Impact Sounds
  (Utilities|IsValid Player
    (:Is%20Valid
      (Networking|CanExecuteCosmeticEvents
        (:True
          (if (HelperFunctions|IsLocalPawn Player)
            (Utilities|FlowControl|DoOnce false
              (:Completed
                (Audio|Components|Audio|FadeOut (Variables|Sound|GetSFX_RezOut_AC) 0.1 0.0 "Linear" :!class /Script/Engine.AudioComponent)
                (Audio|PlaySound2D (Variables|Sound|GetSFX_RezOut_Stop))
                (Variables|/Script/AudioShapes.AudioShapeComponent|SetFadeOutTime (Variables|Sound|GetAudio%20Shape%20Fade%20Time) _fortaudioshapebox)
                (Shape|Disable _fortaudioshapebox)
                (CallFunction|RestartAudioSFX)
                (Audio|PlaySound2D (Variables|Default|GetSFX_TimeLost_Penalty_1p))
                (Utilities|FlowControl|Delay (Variables|Sound|GetSFX_AlienVO_Delay))
                (Audio|PlaySound2D (Variables|Sound|GetSFX_AlienVO_Death))))
            (else
              (Audio|SpawnSoundatLocation (Variables|Default|GetSFX_TimeLost_Penalty) (Transformation|GetWorldLocation (Variables|/Script/Engine.Character|GetMesh Player))))))))))

(event Custom|Reset%20Modifying%20Sound
  (bind _fortaudioshapebox (Variables|Default|GetFortAudioShapeBox))
  ;[1 0 0.8076] Play Impact Sounds
  (Utilities|FlowControl|DoOnce false
    (:Completed
      (Audio|Components|Audio|FadeOut (Variables|Sound|GetSFX_RezOut_AC) 0.1 0.0 "Linear" :!class /Script/Engine.AudioComponent)
      (Audio|PlaySound2D (Variables|Sound|GetSFX_RezOut_Stop))
      (Variables|/Script/AudioShapes.AudioShapeComponent|SetFadeOutTime (Variables|Sound|GetAudio%20Shape%20Fade%20Time) _fortaudioshapebox)
      (Shape|Disable _fortaudioshapebox)
      (CallFunction|RestartAudioSFX)
      (Audio|PlaySound2D (Variables|Default|GetSFX_TimeLost_Penalty_1p))
      (Utilities|FlowControl|Delay (Variables|Sound|GetSFX_AlienVO_Delay))
      (Audio|PlaySound2D (Variables|Sound|GetSFX_AlienVO_Death)))))

(event OnComponentBeginOverlap(BeginChangingSoundBoxComponent)
  (params (OverlappedComponent :type Primitive%20Component%20Object%20Reference) (OtherActor :type Actor%20Object%20Reference) (OtherComp :type Primitive%20Component%20Object%20Reference) (OtherBodyIndex :type Integer) (bFromSweep :type Boolean) (SweepResult :type Hit%20Result%20Structure%20%28by%20ref%29))
  ;[1 0 0.8393] Bind event to Character movement
  (Networking|CanExecuteCosmeticEvents
    (:True
      (bind _asfort_player_pawn (Utilities|Casting|CastToFortPlayerPawn OtherActor))
      (bind _output_get (Variables|Default|SetPlayer%20Pawn _asfort_player_pawn))
      (Character|BindEventtoOnCharacterMovementUpdated _output_get (EventDispatchers|CreateEvent :!event Custom|Event%20On%20Character%20Movement%20Update))
      (Utilities|FlowControl|DoOnce false
        (:Completed
          (EventDispatchers|BindEventtoOnPlayerEnteredLootRoom (Actor|GetComponentbyClass _output_get "/Script/SkyfireRuntime.SkyfirePlayerPawnComponent") (EventDispatchers|CreateEvent :!event Custom|EventOnEnterLootRoom)))))))

(event Custom|RestartAudioSFX
  ;[1 0 0.5556] Restart Audio SFX
  (Utilities|FlowControl|Delay (Variables|Sound|GetAudio%20Shape%20Disabled%20Duration%20))
  (Shape|Enable (Variables|Default|GetFortAudioShapeBox))
  (bind _player_pawn (Variables|Default|GetPlayer%20Pawn)
    (:then
      (Character|UnbindEventfromOnCharacterMovementUpdated _player_pawn (EventDispatchers|CreateEvent :!event Custom|Event%20On%20Character%20Movement%20Update))
      (Audio|Parameter|SetFloatParameter (Control|GetEdgeAudioComponent (Variables|Default|GetFortAudioShapeBox)) (Variables|Sound|GetSFX_Parameter_Name) (Variables|Sound|GetDefault_Parameter_Value) :!class /Script/AudioExtensions.AudioParameterControllerInterface)
      (CallFunction|ResetRezSpawnSound)
      (CallFunction|ResetModifyingSound))
    (:else)))

(event OnComponentBeginOverlap(EnergyGridImpactVolume)
  (params (OverlappedComponent :type Primitive%20Component%20Object%20Reference) (OtherActor :type Actor%20Object%20Reference) (OtherComp :type Primitive%20Component%20Object%20Reference) (OtherBodyIndex :type Integer) (bFromSweep :type Boolean) (SweepResult :type Hit%20Result%20Structure%20%28by%20ref%29))
  ;[0.4496 0.07855 1] OnComponentBeginOverlay (EnergyGridImpactVolume) - If the server, multicast the impact. If local client, process the impact predicitvely. (We do this since the player will drop collision/go into stasis on impact so it is possible this overlap won't trigger on clients.)
  (if (or (Networking|IsServer) (HelperFunctions|IsLocalPawn OtherActor))
    (bind _asfort_player_pawn (Utilities|Casting|CastToFortPlayerPawn OtherActor))
    (CallFunction|OnEnergyGridImpact _asfort_player_pawn)))

(event Custom|OnEnergyGridImpact
  (params (ImpactingPawn :type Fort%20Player%20Pawn%20Object%20Reference))
  (bind _returnvalue (Ability|GetAbilitySystemComponent ImpactingPawn))
  (bind (_value _returnvalue_2) (Utilities|Map|Find (Variables|Default|GetLocalPlayersLastGridImpactTimeMap) ImpactingPawn))
  (bind _returnvalue_3 (Fortnite|Sync|Time|GetServerWorldTimeSeconds))
  (bind _charactermovement (Variables|/Script/Engine.Character|GetCharacterMovement ImpactingPawn))
  ;[0.7031 0.3465 1] OnEnergyGridImpact - Play Ripple FX when collide with the elimination grid and pull the camera back a bit
  ; If this is a local pawn, since this can be called predicitvely and then again from the server, check the last impact time to prevent it triggering twice
  (Utilities|IsValid ImpactingPawn
    (:Is%20Valid
      (if (HelperFunctions|IsLocalPawn ImpactingPawn)
        (if (or (>= (- _returnvalue_3 _value) (Variables|Default|GetPreventDoubleImpactTime)) (not _returnvalue_2))
          (Utilities|Map|Add (Variables|Default|GetLocalPlayersLastGridImpactTimeMap) ImpactingPawn _returnvalue_3)
          ;;[1 0.06985 0.9805] For local players, apply camera effect
          (Utilities|IsValid _returnvalue
            (:Is%20Valid
              (GameplayEffects|ApplyGameplayEffectToSelf :self _returnvalue :GameplayEffectClass (Variables|Default|GetGridImpactCameraGE) :Level 0.0) ; Adjust our camera mode when we hit the energy grid\nOnly need this on local client
            )
            (:Is%20Not%20Valid
              _)))
        (else
          _))
      (Utilities|FlowControl|Sequence
        (:then_0
          ;;[0.9784 0.0786 1] Play impact effects
          (Networking|CanExecuteCosmeticEvents
            (:True
              (Effects|Components|ParticleSystem|SpawnFXAtLocation (Variables|Default|GetRippleFX) (Transformation|GetActorLocation ImpactingPawn)) ; Plays the ripple VFX when you hit the energy grid
              (bind (_direction _paramvalue) (Timeline|UpdateMaterials PlayFromStart)
                (:Update
                  (Rendering|Material|SetScalarParameterValueonMaterials (Variables|/Script/Engine.StaticMeshActor|GetStaticMeshComponent (Variables|Default|GetEnergyGrid)) (Variables|Default|GetImpactParamName) _paramvalue) ; Plays a flash of light on the energy grid where the collision happened
                )))))
        (:then_1
          ;;[1 0.13 0.745] Set actor hidden and stop movement
          (Rendering|SetActorHiddenInGame ImpactingPawn true)
          (Utilities|FlowControl|Delay (Variables|Default|GetStopMovementDelay)) ; Delay stopping movement a tad to prevent the camera from jerking
          (Utilities|IsValid ImpactingPawn
            (:Is%20Valid
              (Utilities|IsValid _charactermovement
                (:Is%20Valid
                  (Components|Movement|StopMovementImmediately _charactermovement) ; Also helps prevent the camera from dipping into the elimination grid
                )))))))))

(event Custom|EventOnEnterLootRoom
  (params (SkyfirePawnComponent :type Skyfire%20Player%20Pawn%20Component%20Object%20Reference))
  ;[1 0.1071 0.9611] Force Fade Out the Rez Out sound, Unbind event after use.
  (Audio|Components|Audio|FadeOut (Variables|Sound|GetSFX_RezOut_AC) 0.1 0.0 "Linear" :!class /Script/Engine.AudioComponent)
  (EventDispatchers|UnbindEventfromOnPlayerEnteredLootRoom SkyfirePawnComponent (EventDispatchers|CreateEvent :!event Custom|EventOnEnterLootRoom)))

(fn ConstructionScript)
