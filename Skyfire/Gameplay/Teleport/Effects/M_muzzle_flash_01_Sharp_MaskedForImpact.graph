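# Unlit, two-sided translucent particle material in the teleport effects folder that draws a sharp muzzle-flash shape for sprites, mesh particles and ribbons.
# A dynamic parameter blends in a panning tornado-pattern breakup and controls its speed and tiling.
# A static switch places that shape in opacity or emissive, both driven by particle color.
# Opacity fades near the camera, edge-on and at depth intersections, except on low quality.
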Material /Skyfire/Gameplay/Teleport/Effects/M_muzzle_flash_01_Sharp_MaskedForImpact
graph {
 properties {
  bUsedWithMeshParticles=true
  bUsedWithNiagaraMeshParticles=true
  bUsedWithNiagaraRibbons=true
  bUsedWithNiagaraSprites=true
  bUsedWithParticleSprites=true
  blendMode=BLEND_Translucent
  opacityMaskClipValue=0.05
  pixelDepthOffsetMode=PDOM_Legacy
  shadingModel=MSM_Unlit
  twoSided=true
  usedShadingModels=Unlit
 }
 nodes [
  n1 = ParticleColor()
  n2 = TextureSampleParameter2D() {parameterName=FlashTexture, texture=/Game/Weapons/FORT_Pistols/Effects/Muzzle/T_Muzzle_Flash_01_Sharp}
  n3 = DynamicParameter() {defaultValue=<rgba 0 1 1 1>, paramNames [TextureBlend,PanSpeed,VTiling,UTiling]}
  n4 = Panner(
   Coordinate=Multiply(A=TextureCoordinate() {uTiling=2, vTiling=0.5}, B=AppendVector(A=@n3.VTiling, B=@n3.VTiling))
   Time=Multiply(A=@n3.PanSpeed, B=Time())
  ) {speedY=1}
  n5 = LinearInterpolate(
   A=@n2.R
   B=Multiply(A=@n2.R, B=TextureSample(UVs=@n4) {texture=/Game/Global/Textures/Default/Tileable_Patterns/T_Tornado_Tile}.R)
   Alpha=@n3.TextureBlend
  )
  MaterialFunctionCall(Emissive=StaticSwitchParameter(True=@n1.RGB, False=Multiply(A=@n5, B=@n1.RGB)) {defaultValue=true, parameterName=Is Translucent}) {materialFunction=&MF_FX_EmissiveAdjustments}
  n6 = ScalarParameter() {defaultValue=200, parameterName=NearFadeDist}
  n7 = DotProduct(A=CameraVectorWS(), B=VertexNormalWS())
  n8 = Multiply(A=Power(Base=Abs(None=@n7))) {constB=4}
  n9 = DepthFade(
   Opacity=Multiply(
    A=MaterialFunctionCall(Distance=@n6) {materialFunction=&Near_Camera_Fade}.Result
    B=Clamp(None=@n8)
   )
   FadeDistance=ScalarParameter() {defaultValue=16, parameterName=DepthFadeDist}
  )
  n10 = QualitySwitch(Default=@n9, Low=Constant() {r=1}, High=@n9, Medium=@n9)
  StaticSwitchParameter(True=Multiply(A=@n5, B=Multiply(A=@n1.A, B=@n10)), False=Multiply(A=Multiply(A=@n2.R, B=@n1.A), B=@n10)) {defaultValue=true, parameterName=Is Translucent}
 ]
}
