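# Niagara system for the player impact effect of Skyfire Grid teleport, layering five cyan-tinted burst emitters with no user parameters.
# A brief double-rendered sprite flash pairs with short-lived jet meshes that spawn repeatedly and shrink over the emitter loop.
# Two mesh wave emitters expand through scale curves while dynamic material parameters and color curves animate their look and fade.
# Sprite wisps spawn on a sphere surface, push outward, and drift upward under acceleration and drag.
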NiagaraSystem /Skyfire/Gameplay/Teleport/Effects/NS_Skyfire_Grid_Player_Impact
dependencies {
 usedDataInterfaces [&NiagaraDataInterfaceCurve, &NiagaraDataInterfaceColorCurve, &NiagaraDataInterfaceVectorCurve]
 usedDynamicInputs []
 usedModules [&SystemState]
}
emitters {
 FLASH001 {
  emitterUpdateScript [
   &EmitterState()
   &SpawnBurst_Instantaneous(Spawn Count=1, Use Spawn Probability=false)
  ]
  particleSpawnScript [
   /Niagara/Modules/Spawn/Initialization/InitializeParticle(
    Color=<rgba 3.6872 20.3822 25 1>
    Lifetime=0.125
    Module.Write Color=true
    Module.Write Lifetime=true
    Module.Write Mass=true
    Module.Write Material Random=true
    Module.Write Position=false
    Module.Write Scale=false
    Module.Write SpriteRotation=true
    Module.Write SpriteSize=true
    Position=&SimulationPosition()
    Sprite Rotation=&RandomRangeFloat(Minimum=0, Maximum=360)
    Sprite Size=<1000 1000>
   )
   &CameraOffset(Camera Offset Amount=15)
   /SetVariables(Particles.SpriteFacing=<0 0 1>) {isSetParameters=true}
   /SetVariables(Particles.CamOffset2=50) {isSetParameters=true}
  ]
  particleUpdateScript [
   &ParticleState()
   &ScaleColor(
    Scale Alpha=&FloatFromCurve(FloatCurve=DataInterface() {propertyValues {Curve {keys [{arriveTangent=-2.9556, interpMode=RCIM_Cubic, leaveTangent=-2.9556, tangentMode=RCTM_User, time=0, value=1}, <at 1 0>]}}})
    ScaleA=true
    ScaleRGB=true
   )
   &ScaleSpriteSize(Scale Factor=&Vector2DFromFloat(Value=&FloatFromCurve(FloatCurve=DataInterface() {propertyValues {Curve {keys [
          {arriveTangent=10.3957, interpMode=RCIM_Cubic, leaveTangent=10.3957, tangentMode=RCTM_User, time=0, value=0}
          <at 0.215 1 RCIM_Cubic>
          <at 1 0>
         ]}}})))
   &SolveForcesAndVelocity(Clamp Velocity=false, Limit Acceleration=false)
  ]
  rendererClasses [&NiagaraSpriteRendererProperties]
  renderers [
   {rendererClass=&NiagaraSpriteRendererProperties}
   {rendererClass=&NiagaraSpriteRendererProperties, rendererIndex=1}
  ]
 }
 Jet_Looping001 {
  emitterUpdateScript [
   &EmitterState(Inactive Response=::NewEnumerator1, Life Cycle Mode=::NewEnumerator1, Loop Behavior=::NewEnumerator1, Loop Duration=0.25, UseLoopDelay=false)
   &SpawnRate(SpawnRate=15, Use Spawn Probability=false)
   &SpawnBurst_Instantaneous(Spawn Count=1, Use Spawn Probability=false)
  ]
  particleSpawnScript [
   /Niagara/Modules/Spawn/Initialization/V2/InitializeParticle(
    Color=<rgba 1.875 14.5937 25 1>
    Lifetime Max=0.35
    Lifetime Min=0.15
    Lifetime Mode=::NewEnumerator1
    Mesh Scale=&Multiply_Vector(
     A=&RandomRangeVector(Minimum=<1.1 0.65 0.65>, Maximum=<1 1 1>)
     B=&VectorFromFloat(Value=0.45)
    )
    Mesh Scale Mode=::NewEnumerator4
    Position Offset Coordinate Space=::World
    UsePositionOffset=true
    Write Material Random=true
   ) {module=InitializeParticle001}
   &InitialMeshOrientation(
    Orientation Coordinate Space=::Local
    Orientation Vector=&MaskVectorBySpawnGroup(Masked Vector=<0 0 1>, Passthrough Vector=<0 0 1>)
    Rotation=&RandomRangeVector(Minimum=<0 0 0>, Maximum=<1 0 0>)
    Use Orientation Vector=true
    Use Rotation Vector=true
   )
   &SystemLocation(MaskSpawnGroup=false, Offset=<0 0 -75>)
   &CameraOffset(Camera Offset Amount=35)
  ]
  particleUpdateScript [
   &ParticleState()
   &ScaleMeshSize(Scale Factor=&Multiply_Vector(
     A=&VectorFromCurve(
      VectorCurve=DataInterface() {propertyValues {
        XCurve {keys [
          {
           arriveTangent=19.9864
           arriveTangentWeight=1.7266
           interpMode=RCIM_Cubic
           leaveTangent=19.9864
           leaveTangentWeight=1.7266
           tangentMode=RCTM_User
           tangentWeightMode=RCTWM_WeightedBoth
           time=0
           value=2
          }
          {arriveTangent=-0.7098, interpMode=RCIM_Cubic, leaveTangent=-0.7098, tangentMode=RCTM_User, time=0.7, value=2.5}
          {arriveTangent=-4.2135, interpMode=RCIM_Cubic, leaveTangent=-4.2135, tangentMode=RCTM_User, time=1, value=0}
         ]}
        YCurve {keys [
          {
           arriveTangent=19.9267
           arriveTangentWeight=2.5558
           interpMode=RCIM_Cubic
           leaveTangent=19.9267
           leaveTangentWeight=2.5558
           tangentMode=RCTM_User
           tangentWeightMode=RCTWM_WeightedBoth
           time=0
           value=2
          }
          <at 1 0>
         ]}
        ZCurve {keys [
          {
           arriveTangent=3.9969
           arriveTangentWeight=0.5658
           interpMode=RCIM_Cubic
           leaveTangent=3.9969
           leaveTangentWeight=0.5658
           tangentMode=RCTM_User
           tangentWeightMode=RCTWM_WeightedBoth
           time=0
           value=2
          }
          {arriveTangent=-1.2101, interpMode=RCIM_Cubic, leaveTangent=-1.2101, tangentMode=RCTM_User, time=1.0016, value=0.28}
         ]}
       }}
      Scale Curve=<2 2.5 2.5>
     )
     B=&VectorFromFloat(Value=&FloatFromCurve(
       FloatCurve=DataInterface() {propertyValues {Curve {keys [
           <at 0 1 RCIM_Cubic>
           {arriveTangent=-1, interpMode=RCIM_Cubic, leaveTangent=-1, time=0.8, value=0.95}
           {arriveTangent=-2.9562, interpMode=RCIM_Cubic, leaveTangent=-2.9562, tangentMode=RCTM_User, time=1, value=0}
          ]}}}
       CurveIndex=$Emitter.NormalizedLoopAge
      ))
    ))
  ]
  rendererClasses [&NiagaraMeshRendererProperties]
  renderers [{rendererClass=&NiagaraMeshRendererProperties, rendererIndex=0}]
 }
 NE_BaseWave {
  emitterUpdateScript [
   &EmitterState()
   &SpawnBurst_Instantaneous(Spawn Count=1, Use Spawn Probability=false)
  ]
  particleSpawnScript [&InitializeParticle(
    Color=<rgba 1 1 1 0.035>
    Lifetime=1.5
    Mesh Scale=&Multiply_Vector(A=<2.5 2.5 0.3>, B=&VectorFromFloat(Value=5.25))
    Mesh Scale Mode=::NewEnumerator4
    UsePositionOffset=false
    Write Material Random=true
   ) {module=InitializeParticle001}]
  particleUpdateScript [
   &ParticleState()
   &SolveForcesAndVelocity(Clamp Velocity=false, Limit Acceleration=false)
   &ScaleMeshSize(Scale Factor=&VectorFromFloat(Value=&FloatFromCurve(FloatCurve=DataInterface() {propertyValues {Curve {keys [
          {arriveTangent=2.4929, interpMode=RCIM_Cubic, leaveTangent=2.4929, tangentMode=RCTM_User, time=0, value=0}
          {interpMode=RCIM_Cubic, tangentMode=RCTM_User, time=1, value=1}
         ]}}})))
   &DynamicMaterialParameters(
    Index 0 Param 1=&FloatFromCurve(FloatCurve=DataInterface() {propertyValues {Curve {keys [<at 0 0>,<at 1 0.75>]}}})
    Index 0 Param 2=&FloatFromCurve(FloatCurve=DataInterface() {propertyValues {Curve {keys [
         <at 0 0 RCIM_Cubic>
         {arriveTangent=16.6341, interpMode=RCIM_Cubic, leaveTangent=16.6341, tangentMode=RCTM_User, time=0.034, value=3.766}
         {arriveTangent=-22.3955, interpMode=RCIM_Cubic, leaveTangent=-21.075, tangentMode=RCTM_Break, time=0.14, value=3.6437}
         <at 0.2 2.5>
         {arriveTangent=-8.5799, interpMode=RCIM_Cubic, leaveTangent=-6.7449, tangentMode=RCTM_Break, time=0.26, value=1.6045}
         <at 0.65 0.2787>
        ]}}})
    Index 0 Param 4=1
    Module.Param0WriteEnabled=true
    Module.Param1WriteEnabled=true
    Module.Param2WriteEnabled=true
    Module.Param3WriteEnabled=true
   )
   &ScaleColor(
    Linear Color Curve=DataInterface() {propertyValues {
      AlphaCurve {keys [<at 0.0574 0.5 RCIM_Cubic>,<at 1 0 RCIM_Cubic>]}
      BlueCurve {keys [<at 0.0753 1 RCIM_Cubic>]}
      GreenCurve {keys [
        <at 0.0753 0 RCIM_Cubic>
        {arriveTangent=0.3305, interpMode=RCIM_Cubic, leaveTangent=0.3305, time=0.5279, value=0.8167}
        <at 0.9866 0.3012 RCIM_Cubic>
       ]}
      RedCurve {keys [<at 0.0753 0 RCIM_Cubic>]}
     }}
    Scale Mode=::NewEnumerator2
   ) {module=ScaleColor001}
  ]
  rendererClasses [&NiagaraMeshRendererProperties]
  renderers [{rendererClass=&NiagaraMeshRendererProperties, rendererIndex=0}]
 }
 Wave02 {
  emitterUpdateScript [
   &EmitterState(Life Cycle Mode=::NewEnumerator1, Loop Behavior=::NewEnumerator1, Loop Duration=0.5, UseLoopDelay=false)
   &SpawnBurst_Instantaneous(Spawn Count=2, Use Spawn Probability=false)
  ]
  particleSpawnScript [
   &InitializeParticle(Lifetime=4, Mesh Scale Mode=::NewEnumerator3, Mesh Uniform Scale=5, UsePositionOffset=false, Write Material Random=true)
   &InitialMeshOrientation(
    Orientation Axis=<0 0 1>
    Orientation Vector=<0 0 1>
    Rotation=&RandomRangeVector(Minimum=<0 0 -1>, Maximum=<0 0 1>)
    Use Orientation Vector=true
    Use Rotation Vector=true
   )
   &CameraOffset(Camera Offset Amount=15)
   &SystemLocation(MaskSpawnGroup=false, Offset=<0 0 10>)
  ]
  particleUpdateScript [
   &ParticleState()
   &KillParticles(Kill Particles=&SetBoolByFloatComparison(A=$Particles.NormalizedAge, B=0.3936, Comparison Type=::NewEnumerator2))
   &ScaleColor(
    Linear Color Curve=DataInterface() {propertyValues {
      AlphaCurve {keys [<at 0 0>,<at 0.0695 0.5>,<at 0.3576 0.5>,<at 1 0>]}
      BlueCurve {keys [<at 0.0283 1>,<at 0.1987 1>,<at 0.3149 1>,<at 0.8 1>]}
      GreenCurve {keys [<at 0.0283 0>,<at 0.1987 0.8167>,<at 0.3149 0.3012>,<at 0.8 0.0414>]}
      RedCurve {keys [<at 0.0283 0>,<at 0.1987 0>,<at 0.3149 0>,<at 0.8 0.5686>]}
     }}
    Scale Mode=::NewEnumerator2
   )
   &DynamicMaterialParameters(
    Index 0 Param 1=&FloatFromCurve(FloatCurve=DataInterface() {propertyValues {Curve {keys [{arriveTangent=-4.0141, interpMode=RCIM_Cubic, leaveTangent=-4.0141, tangentMode=RCTM_User, time=0, value=0.6}, <at 1 -0.7>]}}})
    Index 0 Param 2=&FloatFromCurve(FloatCurve=DataInterface() {propertyValues {Curve {keys [{arriveTangent=-0.5818, interpMode=RCIM_Cubic, leaveTangent=-0.5818, tangentMode=RCTM_User, time=0, value=1}, <at 1 0.9>]}}})
    Index 0 Param 3=&FloatFromCurve(FloatCurve=DataInterface() {propertyValues {Curve {keys [<at 0 0>,<at 0.15 1>,<at 1 1.2>]}}})
    Module.Param0WriteEnabled=true
    Module.Param1WriteEnabled=true
    Module.Param2WriteEnabled=true
    Module.Param3WriteEnabled=false
   )
   &ScaleMeshSize(Scale Factor=&VectorFromFloat(Value=&FloatFromCurve(FloatCurve=DataInterface() {propertyValues {Curve {keys [<at 0 0.25>,<at 0.0596 1.6>,<at 0.3936 2.125>]}}})))
  ]
  rendererClasses [&NiagaraMeshRendererProperties]
  renderers [{rendererClass=&NiagaraMeshRendererProperties, rendererIndex=0}]
 }
 Whispies {
  emitterUpdateScript [
   &EmitterState()
   &SpawnBurst_Instantaneous(Spawn Count=5, Use Spawn Probability=false)
  ]
  particleSpawnScript [
   /Niagara/Modules/Spawn/Initialization/InitializeParticle(
    Color=<rgba 0.6855 1.371 1.1768 1>
    Module.Write Color=true
    Module.Write Lifetime=true
    Module.Write Mass=true
    Module.Write Material Random=true
    Module.Write Position=false
    Module.Write Scale=false
    Module.Write SpriteRotation=true
    Module.Write SpriteSize=true
    Position=&SimulationPosition()
    Sprite Rotation=&RandomRangeFloat(Minimum=0, Maximum=360)
    Sprite Size=&Lerp_Vector2(A=<160 160>, B=<120 120>, Alpha=$Particles.Random)
   )
   &SphereLocation(
    MaskSpawnGroup=false
    Offset Coordinate Space:ENiagaraCoordinateSpace=::World
    Override Seed=false
    Surface Only=true
    UseNonUniformScale=false
    UseOffset=true
    UseSphereOrientation=false
   )
   &DynamicMaterialParameters(
    Index 0 Param 1=$Particles.Random
    Index 0 Param 2=&RandomRangeFloat(Minimum=0, Maximum=1)
    Index 0 Param 3=$Particles.Random
    Module.Param0WriteEnabled=true
    Module.Param1WriteEnabled=true
    Module.Param2WriteEnabled=true
    Module.Param3WriteEnabled=true
   ) {module=DynamicMaterialParameters001}
   &CameraOffset(Camera Offset Amount=&RandomRangeFloat(Minimum=-25, Maximum=25))
   &AddVelocityFromPoint(
    Module.Use Falloff Distance=false
    Origin Offset=<0 0 75>
    Velocity Strength=&RandomRangeFloat(Minimum=100, Maximum=25)
   )
  ]
  particleUpdateScript [
   &ParticleState()
   &DynamicMaterialParameters(
    Index 0 Param 1=&RandomRangeFloat(Minimum=0.2, Maximum=1.1)
    Index 0 Param 2=&RandomRangeFloat(Minimum=2, Maximum=1)
    Index 0 Param 4=-1
    Module.Param0WriteEnabled=true
    Module.Param1WriteEnabled=true
    Module.Param2WriteEnabled=true
    Module.Param3WriteEnabled=true
   )
   &ScaleSpriteSize(Scale Factor=&Vector2DFromFloat(Value=&FloatFromCurve(
      FloatCurve=DataInterface() {propertyValues {Curve {keys [
          {arriveTangent=3.8006, interpMode=RCIM_Cubic, leaveTangent=3.8006, tangentMode=RCTM_User, time=0, value=0}
          {arriveTangent=1, interpMode=RCIM_Cubic, leaveTangent=1, time=0.3065, value=0.764}
          <at 1 1>
         ]}}}
      Scale Curve=2.5
     )))
   &ScaleColor(
    Scale Mode=::NewEnumerator1
    Scale RGBA=&MakeVector4_FromLinearColor(LinearColor=&ColorFromCurve(ColorCurve=DataInterface() {propertyValues {
        AlphaCurve {keys [<at 0 0 RCIM_Cubic>,<at 0.3476 0.5>,<at 1 0 RCIM_Cubic>]}
        BlueCurve {keys [<at 0 2 RCIM_Cubic>,<at 0.8587 1 RCIM_Cubic>]}
        GreenCurve {keys [<at 0 1 RCIM_Cubic>,<at 0.8587 1 RCIM_Cubic>]}
        RedCurve {keys [<at 0 1 RCIM_Cubic>,<at 0.8587 1 RCIM_Cubic>]}
       }}))
    ScaleRGBA=true
   )
   &AccelerationForce(Acceleration=<0 0 100>)
   &Drag(Use Linear Drag=true, Use Rotational Drag=false)
   &SolveForcesAndVelocity(Clamp Velocity=false, Limit Acceleration=false)
  ]
  rendererClasses [&NiagaraSpriteRendererProperties]
  renderers [{rendererClass=&NiagaraSpriteRendererProperties}]
 }
}
summary {
 systemName=NS_Skyfire_Grid_Player_Impact
 emitters [
  {emitterName=Whispies, rendererClasses [&NiagaraSpriteRendererProperties]}
  {emitterName=FLASH001, rendererClasses [&NiagaraSpriteRendererProperties]}
  {emitterName=Wave02, rendererClasses [&NiagaraMeshRendererProperties]}
  {emitterName=Jet_Looping001, rendererClasses [&NiagaraMeshRendererProperties]}
  {emitterName=NE_BaseWave, rendererClasses [&NiagaraMeshRendererProperties]}
 ]
 userVariables []
}
