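; GA_Skyfire_BlockShooting is a Gameplay Ability in the Skyfire teleport gameplay folder, named for blocking shooting.
; Its graph holds only an ActivateAbility event with no connected logic, so activation runs no Blueprint steps and never calls EndAbility.
; Any blocking effect is not visible in this graph.
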(event Ability|EventActivateAbility)
