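; Gameplay ability for the Skyfire grid impact that grants a temporary camera angle before the warp tunnel starts.
; On activation, where cosmetic events can execute, it pushes the configured camera mode, waits for the configured duration, then ends itself.
; Ending the ability pops the camera mode, whether or not it was cancelled.
; The non-cosmetic path has no delay or end call of its own.
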(event Ability|EventActivateAbility
  ;[0.8007 0.01503 1] Grant a temporary camera angle before we start to warp tunnel
  (|CallParentFunction)
  (Networking|CanExecuteCosmeticEvents
    (:True
      (Camera|PushCameraMode (Variables|Default|GetTemporaryCameraAngle))
      (Utilities|FlowControl|Delay (Variables|Default|GetTemporaryCameraAngleDuration)) ; Time for this new camera to persist before we start to warp tunnel
      (Ability|EndAbility))))

(event Ability|EventOnEndAbility
  (params (bWasCancelled :type Boolean))
  ;[0.8007 0.01503 1] Grant a temporary camera angle before we start to warp tunnel
  (|CallParentFunction bWasCancelled)
  (Camera|PopCameraMode))
