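; Skyfire gameplay ability that plays the post-teleport prison wakeup montage while temporarily hiding the player's weapon and backpack.
; Once required character parts are ready, it holsters the weapon, hides the backpack (including late-equipped ones), delays, then plays the montage under a tracking GameplayEffect.
; Counted montage triggers reveal backpack, then weapon.
; A failsafe timer can end the ability; ending removes the effect and force-shows both items.
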(event Ability|EventActivateAbility
  ;[0.03212 1 0] Play the CellBlock Montage
  (|CallParentFunction)
  (bind _returnvalue (Utilities|Time|SetTimerbyEvent (EventDispatchers|CreateEvent :!event Custom|FailsafeTimer) (ScalableFloat|GetValueatLevel (Variables|Default|GetFailsafeTime))))
  (Variables|Default|SetFailsafeTimerHandle _returnvalue)
  (Variables|Default|SetBackpackVisible false)
  (Variables|Default|SetWeaponVisible false)
  (Utilities|FlowControl|DoOnce false
    (:Completed
      (Ability|Tasks|WaitForCustomCharacterPartsReady (Utilities|Array|MakeArray (Variables|Default|GetRequiredCharacterPart))
        (:OnReady
          (CallFunction|HideWeaponandBackpack)
          (Utilities|FlowControl|Delay (ScalableFloat|GetValueatLevel (Variables|Default|GetArrivalAnimDelay))) ; Wait for player to settle from teleport
          (Ability|Tasks|PlayMontageandWaitforTrigger "None" (Variables|Default|GetMontage%20to%20Play)
            (:then
              (bind _returnvalue_2 (Ability|ApplyGameplayEffectToOwner "/Skyfire/Gameplay/Teleport/GE_Skyfire_PrisonAnimationPlaying.GE_Skyfire_PrisonAnimationPlaying_C")) ; Apply GE with tag to track the Montage being played.
              (Variables|Default|SetMontageHandle _returnvalue_2))
            (:Triggered
              (bind _result (Math|Integer|IncrementInt (Variables|Default|GetTriggerCount))
                (:%20%20
                  (if (> _result (Variables|Default|GetTriggerCountThreshold))
                    (CallFunction|ShowWeapon)
                    (else
                      (CallFunction|ShowBackpack))))))
            (:Cancelled
              _)
            (:Completed
              _))
          (Utilities|Time|ClearandInvalidateTimerbyHandle (Variables|Default|GetFailsafeTimerHandle))
          (Ability|EndAbility))))) ; Failsafe in case we somehow we  re-call this GA since this GA persists through all skyfire phases
)

(event Custom|FailsafeTimer
  ;[0.03212 1 0] Play the CellBlock Montage
  (Ability|EndAbility))

(event Ability|EventOnEndAbility
  (params (bWasCancelled :type Boolean))
  ;[0 0.1529 1] Clean up the applied GE
  (|CallParentFunction bWasCancelled)
  (Ability|RemoveGameplayEffectFromOwnerWithHandle (Variables|Default|GetMontageHandle))
  ;;[1 0.9082 0.1812] Force show (i.e. un-holster) the weapon and backpack to account for any timing/loading/network issues that may have prevented them from being visible.
  (CallFunction|ShowWeapon)
  (CallFunction|ShowBackpack))

(event Custom|HideWeaponAndBackpack
  ;[0.9848 0.03478 1] Hide the Knockback Weapon and Backpack Meshes
  (bind _outfortplayerpawn (CallFunction|GetAvatarasValidFortPlayerPawn)
    (:ValidFortPlayerPawn
      (bind _output_get (Variables|Default|SetPlayerPawn _outfortplayerpawn))
      (Utilities|IsValid _output_get
        (:Is%20Valid
          (Weapon|SetHolsterWeapon _output_get 0 true false (Variables|Default|GetHolster%20Id))
          (Utilities|IsValid (Customization|GetSkeletalMeshforPartType _output_get "Backpack" :!class /Script/FortniteGame.FortPlayerPawn)
            (:Is%20Valid
              (Rendering|SetVisibility (Customization|GetSkeletalMeshforPartType _output_get "Backpack" :!class /Script/FortniteGame.FortPlayerPawn) false true))
            (:Is%20Not%20Valid
              (EventDispatchers|AssignOnPartChanged _output_get (EventDispatchers|CreateEvent :!event BackPackEquipped)))))))))

(event Custom|BackpackEquipped
  (params (PartType :type EFortCustomPartType%20Enum :default "Head"))
  (bind _playerpawn (Variables|Default|GetPlayerPawn))
  (bind _returnvalue (Customization|GetSkeletalMeshforPartType _playerpawn "Backpack" :!class /Script/FortniteGame.FortPlayerPawn))
  ;[0.9848 0.03478 1] Hide the Knockback Weapon and Backpack Meshes
  (if (Utilities|Enum|Equal(Enum) PartType "Backpack")
    (Utilities|IsValid _playerpawn
      (:Is%20Valid
        (Utilities|IsValid _returnvalue
          (:Is%20Valid
            (Rendering|SetVisibility _returnvalue false true)))))))

(event Custom|ShowWeapon
  (bind _playerpawn (Variables|Default|GetPlayerPawn))
  ;[1 0 0.8252] Show the Knockback Weapon and Backpack after materialize VFX play
  (Utilities|IsValid _playerpawn
    (:Is%20Valid
      (Weapon|SetHolsterWeapon _playerpawn 0 false false (Variables|Default|GetHolster%20Id))
      (Variables|Default|SetWeaponVisible true))))

(event Custom|ShowBackpack
  (bind _returnvalue (Customization|GetSkeletalMeshforPartType (Variables|Default|GetPlayerPawn) "Backpack" :!class /Script/FortniteGame.FortPlayerPawn))
  (bind _playerpawn (Variables|Default|GetPlayerPawn))
  ;[1 0 0.8252] Show the Knockback Weapon and Backpack after materialize VFX play
  (Utilities|IsValid _playerpawn
    (:Is%20Valid
      (Utilities|IsValid _returnvalue
        (:Is%20Valid
          (Rendering|SetVisibility _returnvalue true true)
          (Variables|Default|SetBackpackVisible true))))))
